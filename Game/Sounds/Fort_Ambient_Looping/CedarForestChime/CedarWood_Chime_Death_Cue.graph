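# Sound cue for the death of a cedar forest wind chime, layering a destruction sound over an impact hit.
# A mixer combines a randomly chosen death sample, weighted much louder, with a quieter randomly chosen impact sample.
# It plays through the impact sound class with the chime impact attenuation.
# An overridden concurrency limits instances per owner, stopping the oldest first.
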SoundCue /Game/Sounds/Fort_Ambient_Looping/CedarForestChime/CedarWood_Chime_Death_Cue
root=Mixer_1
nodes [
 {id=Mixer_1, children [Random_2,Random_3], properties {inputVolume [3,0.5]}}
 {id=Random_2, children [WavePlayer_11, WavePlayer_12, WavePlayer_13], properties {bShouldExcludeFromBranchCulling=true, weights [1,1,1]}}
 {
  id=Random_3
  children [WavePlayer_14, WavePlayer_15, WavePlayer_16, WavePlayer_17, WavePlayer_18, WavePlayer_19]
  properties {bShouldExcludeFromBranchCulling=true, weights [1,1,1,1,1,1]}
 }
 {id=WavePlayer_11, properties {soundWaveAssetPtr=&/CedarWoods_Chimes_Death_01}}
 {id=WavePlayer_12, properties {soundWaveAssetPtr=&/CedarWoods_Chimes_Death_03}}
 {id=WavePlayer_13, properties {soundWaveAssetPtr=&/CedarWoods_Chimes_Death_02}}
 {id=WavePlayer_14, properties {soundWaveAssetPtr=&/CedarWoods_Chimes_Impact_01}}
 {id=WavePlayer_15, properties {soundWaveAssetPtr=&/CedarWoods_Chimes_Impact_02}}
 {id=WavePlayer_16, properties {soundWaveAssetPtr=&/CedarWoods_Chimes_Impact_03}}
 {id=WavePlayer_17, properties {soundWaveAssetPtr=&/CedarWoods_Chimes_Impact_04}}
 {id=WavePlayer_18, properties {soundWaveAssetPtr=&/CedarWoods_Chimes_Impact_05}}
 {id=WavePlayer_19, properties {soundWaveAssetPtr=&/CedarWoods_Chimes_Impact_06}}
]
properties {
 attenuationSettings=&/CedarWoods_Chime_Impact_Att
 bOverrideConcurrency=true
 maxDistance=4800
 soundClassObject=&Fort_Impacts
 volumeMultiplier=2.5
 concurrencyOverrides {bLimitToOwner=true, maxCount=5, resolutionRule=StopOldest, volumeScaleAttackTime=2}
}
