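# Sound cue for Cedar Forest chime impacts, played in the impacts sound class with its own attenuation.
# A Mixer layers a random short chime, faded out by an Enveloper volume curve, with a random impact sample mixed at higher volume.
# Overridden concurrency caps simultaneous plays per owner and stops the oldest when the limit is reached.
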SoundCue /Game/Sounds/Fort_Ambient_Looping/CedarForestChime/CedarWood_Chime_Impact_Cue
root=Mixer_0
nodes [
 {
  id=Enveloper_0
  children [Random_0]
  properties {pitchCurve {editorCurveData {keys [{time=0}]}}, volumeCurve {editorCurveData {keys [<at 0 1>,<at 0.5 0.6>,<at 1 0>]}}}
 }
 {id=Mixer_0, children [Enveloper_0,Random_1], properties {inputVolume [1,2]}}
 {
  id=Random_0
  children [WavePlayer_0,WavePlayer_1,WavePlayer_2,WavePlayer_3,WavePlayer_4]
  properties {bShouldExcludeFromBranchCulling=true, weights [1,1,1,1,1]}
 }
 {
  id=Random_1
  children [WavePlayer_5, WavePlayer_6, WavePlayer_7, WavePlayer_8, WavePlayer_9, WavePlayer_10]
  properties {bShouldExcludeFromBranchCulling=true, weights [1,1,1,1,1,1]}
 }
 {id=WavePlayer_0, properties {soundWaveAssetPtr=&/CedarWoods_Chimes_Med_Short_05}}
 {id=WavePlayer_1, properties {soundWaveAssetPtr=&/CedarWoods_Chimes_Med_Short_01}}
 {id=WavePlayer_10, properties {soundWaveAssetPtr=&/CedarWoods_Chimes_Impact_06}}
 {id=WavePlayer_2, properties {soundWaveAssetPtr=&/CedarWoods_Chimes_Med_Short_02}}
 {id=WavePlayer_3, properties {soundWaveAssetPtr=&/CedarWoods_Chimes_Med_Short_03}}
 {id=WavePlayer_4, properties {soundWaveAssetPtr=&/CedarWoods_Chimes_Med_Short_04}}
 {id=WavePlayer_5, properties {soundWaveAssetPtr=&/CedarWoods_Chimes_Impact_01}}
 {id=WavePlayer_6, properties {soundWaveAssetPtr=&/CedarWoods_Chimes_Impact_02}}
 {id=WavePlayer_7, properties {soundWaveAssetPtr=&/CedarWoods_Chimes_Impact_03}}
 {id=WavePlayer_8, properties {soundWaveAssetPtr=&/CedarWoods_Chimes_Impact_04}}
 {id=WavePlayer_9, properties {soundWaveAssetPtr=&/CedarWoods_Chimes_Impact_05}}
]
properties {
 attenuationSettings=&/CedarWoods_Chime_Impact_Att
 bOverrideConcurrency=true
 maxDistance=4800
 soundClassObject=&Fort_Impacts
 volumeMultiplier=2
 concurrencyOverrides {bLimitToOwner=true, maxCount=5, resolutionRule=StopOldest, volumeScaleAttackTime=2}
}
